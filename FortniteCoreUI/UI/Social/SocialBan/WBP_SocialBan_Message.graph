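# WBP_SocialBan_Message is a CommonUserWidget showing an alert icon and wrapped text telling players how long a community-rules ban blocks chat or friend adding.
# SetBannedTimeLeft formats the timespan as approximate text, choosing the chat or friends message via UseFriendsSpecificMessage.
# PreConstruct adjusts icon size and padding for compact UI, and previews a placeholder duration at design time.
# Construct feeds the owning player's social user into FortSocialUserInfoViewModel.
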WidgetBlueprint /FortniteCoreUI/UI/Social/SocialBan/WBP_SocialBan_Message
graphs [
 EventGraph {nodes [
   n1 = Event PreConstruct()
   > SetCompactUIStyle() {is=title K2Node_CallFunction}
   > Branch(Condition=@n1.IsDesignTime)
   > SetBannedTimeLeft!(
    execute
    self
    TimeLeft:Timespan Structure=MakeTimespan!(Days:int=5, Hours:int=0, Minutes:int=0, Seconds:int=0, Milliseconds:int=0) -> (ReturnValue:Timespan Structure)
   ) -> (then)
   Event Construct()
   n2 = GetOwningPlayer()
   > SetSocialUser!(
    execute
    self:Fort Social User Viewmodel Object Reference=Get FortSocialUserInfoViewModel() -> (FortSocialUserInfoViewModel:Fort Social User Info View Model Object Reference)
    InSocialUser:Fort Social User Object Reference=GetSocialUserFromController!(FortPC:Fort Player Controller Object Reference=GetFortPlayerControllerFromActor!(Actor:Actor Object Reference=@n2) -> (ReturnValue:Fort Player Controller Object Reference)) -> (ReturnValue:Fort Social User Object Reference)
   ) -> (then)
  ]}
 SetBannedTimeLeft {nodes [
   n1 = SetBannedTimeLeft!() -> (then, TimeLeft:Timespan Structure) {is=title K2Node_FunctionEntry}
   n2 = GetTimespanAsText!(
    Timespan:"Timespan Structure (by ref)"=@n1
    bShowHoursWhenZero:bool=false
    bShowSeconds:bool=false
    FormatType:ETimespanAsTextFormat Enum=Approximate
    UnitDisplayType:ETimespanUnitDisplayFormat Enum=Full
    bUseSentenceCase:bool=true
   ) -> (ReturnValue:Text)
   > "SetText (Text)"(
    self=Get Common_TextBanMessage() -> (Common_TextBanMessage:Common Text Object Reference)
    InText=Select(
     Option 0:Text=Format Text!(Format="LOCTEXT(\"You can\\'t use chat for {timeleft} because of your ban for breaking community rules\")", timeleft:Text=@n2) -> (Result)
     Option 1:Text=Format Text(Format="LOCTEXT(\"You can\\'t add friends for {timeleft} because of your ban for breaking community rules\")", timeleft=@n2)
     Index=Get UseFriendsSpecificMessage() -> (UseFriendsSpecificMessage:bool)
    ) -> (ReturnValue:Text)
   )
  ]}
 SetCompactUIStyle {nodes [
   SetCompactUIStyle() {is=title K2Node_FunctionEntry}
   > Branch(Condition=IsUsingCompactUI())
   n1 = Get SubState_WarningIcon() -> (SubState_WarningIcon:Image Object Reference)
   > SetDesiredSizeOverride!(execute, self:Image Object Reference=@n1, DesiredSize:Vector 2D Structure=(X=32,Y=32)) -> (then)
   n2 = SlotAsHorizontalBoxSlot!(Widget:Widget Object Reference=@n1) -> (ReturnValue:Horizontal Box Slot Object Reference)
   n3 = Get!(self:Horizontal Box Slot Object Reference=@n2) -> (Padding_Left:float=~, Padding_Top:float=~, Padding_Right:float=~, Padding_Bottom:float=~)
   > SetPadding!(
    execute
    self:Horizontal Box Slot Object Reference=@n2
    InPadding_Left:float=@n3
    InPadding_Top:float=6
    InPadding_Right:float=8
    InPadding_Bottom:float=@n3.Padding_Bottom
   ) -> (then)
  ]}
]
variables [
 IsSocialBannedChanged {type=boolean}
 UseFriendsSpecificMessage {type=boolean}
]
widgets {
 info {inheritedWidgetCount=0, parentClass=&CommonUserWidget, rootWidgetClass=&HorizontalBox, widgetCount=3}
 widgets [
  HorizontalBox_1 {widgetClassPath=&HorizontalBox, widget=HorizontalBox_1, properties {bOverride_Cursor=true}}
  SubState_WarningIcon {
   widgetClassPath=&Image
   bIsVariable=true
   parent=HorizontalBox_1
   widget=SubState_WarningIcon
   properties {
    visibility=SelfHitTestInvisible
    brush {
     bIsDynamicallyLoaded=false
     drawAs=Image
     imageType=FullColor
     mirroring=NoMirror
     resourceName=None
     resourceObject=/FortUILibrary/Textures/MaterialInstances/MI_UIKit_Icon_Alert
     tiling=NoTile
     imageSize=<24 24>
     margin=<ltrb 0 0 0 0>
     outlineSettings {
      bUseBrushTransparency=false
      roundingType=HalfHeightRadius
      width=0
      color {colorUseRule=UseColor_Specified, specifiedColor=<rgba 0 0 0 0>}
      cornerRadii=<0 0 0 0>
     }
     tintColor {colorUseRule=UseColor_Specified, specifiedColor=<rgba 0.9301 1 0.0103 1>}
     uVRegion {bIsValid=false, max=<0 0>, min=<0 0>}
    }
   }
   slot {class=HorizontalBoxSlot, properties {horizontalAlignment=HAlign_Left, verticalAlignment=VAlign_Top}}
  }
  Common_TextBanMessage {
   widgetClassPath=&CommonTextBlock
   bIsVariable=true
   parent=HorizontalBox_1
   widget=Common_TextBanMessage
   properties {
    autoWrapText=true
    lineHeightPercentage=1.2
    mobileFontSizeMultiplier=1.33
    text=You can't use chat for 5 Days because of your ban for breaking community rules
    colorAndOpacity {colorUseRule=UseColor_Specified, specifiedColor=<rgba 0.9301 1 0.0103 1>}
    font {
     bForceMonospaced=false
     bMaterialIsStencil=false
     fontMaterial=null
     fontObject=&HeadingNow
     letterSpacing=10
     monospacedWidth=1
     size=13.5
     skewAmount=0
     typefaceFontName=74 Regular
     outlineSettings {
      bApplyOutlineToDropShadows=false
      bMiteredCorners=false
      bSeparateFillAlpha=false
      outlineMaterial=null
      outlineSize=0
      outlineColor=<rgba 0 0 0 1>
     }
    }
    shadowOffset=<0 0>
   }
   slot {class=HorizontalBoxSlot, properties {padding=<ltrb 8 3 0 0>, size {sizeRule=Fill, value=1}}}
  }
 ]
}
